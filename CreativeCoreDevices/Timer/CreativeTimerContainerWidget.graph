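# Container widget for the Creative Timer device, hosting timer widgets in an overlay inside a safe zone.
# On widget creation it adds the widget to the overlay, filling it when both fullscreen and custom.
# Otherwise the widget sits top-center, pushed down by padding and a render translation.
# UpdateViewModel events cast the MVVM viewmodel to the timer type and forward time, text, style and state flags to it.
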WidgetBlueprint /CreativeCoreDevices/Timer/CreativeTimerContainerWidget
graphs [EventGraph {nodes [
   n1 = Event UpdateViewModel!() -> (
    OutputDelegate
    then
    ViewModel:MVVM Base Viewmodel Object Reference
    TimerDeviceObject:Object Reference
    CurrentTime:int=0
    TotalTime:int=0
    LabelText:"Text (by ref)"
    TextStyle:Name=None
    TimerName:"Text (by ref)"
    bIsPaused:bool
    bIsStarted:bool
    bIsUrgently:bool
   ) {is=title K2Node_Event}
   > n2 = Cast To CreativeTimerViewModel!(execute, Object=@n1.ViewModel) -> (then, CastFailed, AsDevice - Timer View Model:Device - Timer View Model Object Reference) {is=title K2Node_DynamicCast}
   > UpdateViewModel!(
    execute
    self:Device - Timer View Model Object Reference=@n2
    CurrentTime:int=@n1.CurrentTime
    TotalTime:int=@n1.TotalTime
    LabelText:"Text (by ref)"=@n1.LabelText
    TextStyle:Name=@n1.TextStyle
    TimerName:"Text (by ref)"=@n1.TimerName
    bIsPaused:bool=@n1.bIsPaused
    bIsStarted:bool=@n1.bIsStarted
    bIsUrgently:bool=@n1.bIsUrgently
   ) -> (then)
   n3 = Event OnTimerWidgetCreated!() -> (OutputDelegate, then, Widget:User Widget Object Reference, bIsFullscreen:bool, bIsCustomWidget:bool) {is=title K2Node_Event}
   > n4 = AddChildToOverlay!(
    execute
    self:Overlay Object Reference=Get OverlayPanel() -> (OverlayPanel:Overlay Object Reference)
    Content:Widget Object Reference=@n3.Widget
   ) -> (then, ReturnValue:Overlay Slot Object Reference)
   > n5 = Branch(Condition=AND Boolean(A=@n3.bIsFullscreen, B=@n3.bIsCustomWidget))
   > SetHorizontalAlignment(self=@n4, InHorizontalAlignment=HAlign_Fill)
   > SetVerticalAlignment(self=@n4, InVerticalAlignment=VAlign_Fill)
   SetHorizontalAlignment(execute=@n5.else, self=@n4, InHorizontalAlignment=HAlign_Center)
   > SetVerticalAlignment(self=@n4, InVerticalAlignment=VAlign_Top)
   > SetPadding!(execute, self:Overlay Slot Object Reference=@n4, InPadding:Margin Structure=Make Margin(Left=0, Top=90, Right=0, Bottom=0)) -> (then)
   > SetRenderTranslation(self=@n3.Widget, Translation=(X=0,Y=80))
  ]}]
variables []
widgets {
 info {
  inheritedWidgetCount=1
  parentClass=/Script/CreativeCoreDevicesUI.FortCreativeTimerContainerWidget
  rootWidgetClass=&SafeZone
  widgetCount=3
 }
 widgets [
  SafeZone {widgetClassPath=&SafeZone, bIsVariable=true, widget=SafeZone}
  OverlayPanel {widgetClassPath=&Overlay, bIsVariable=true, parent=SafeZone, widget=OverlayPanel, slot {class=SafeZoneSlot}}
  TutorialHighlight {widgetClassPath=&TutorialHighlight}
 ]
}
