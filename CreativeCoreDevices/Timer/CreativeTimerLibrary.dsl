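; Function library for the Creative Timer device that adds and removes the timer UI scene for local players.
; SetupTimerUI adds the scene through the Dynamic UI Manager for each local Fortnite player controller where it is not active; CleanupTimerUI removes it where it is.
; Both run only where cosmetic events can execute and do nothing when the legacy Creative timer is in use.
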(fn CleanupTimerUI
  (bind _returnvalue (WorldSubsystems|GetDynamicUIManager 0))
  (bind _timeruiscene (LocalVariables|GetTimerUIScene))
  (Networking|CanExecuteCosmeticEvents
    (:True
      (if (not (Class|AthenaTimeWidgetBase|UseLegacyTimerinCreative))
        (for _array_element (Fortnite|GetLocalFortPlayerControllers true)
          (if (DynamicUI|IsSceneActive _returnvalue _timeruiscene _array_element)
            (DynamicUI|RemoveScene _returnvalue _timeruiscene _array_element)))))))

(fn SetupTimerUI
  (bind _returnvalue (WorldSubsystems|GetDynamicUIManager 0))
  (bind _timeruiscene (LocalVariables|GetTimerUIScene))
  (Networking|CanExecuteCosmeticEvents
    (:True
      (if (not (Class|AthenaTimeWidgetBase|UseLegacyTimerinCreative))
        (for _array_element (Fortnite|GetLocalFortPlayerControllers true)
          (if (not (DynamicUI|IsSceneActive _returnvalue _timeruiscene _array_element))
            (DynamicUI|AddScene _returnvalue _timeruiscene _array_element)))))))
