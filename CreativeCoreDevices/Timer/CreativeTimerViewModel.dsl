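; View model for the Creative Timer device that holds timer state and display strings, presumably for the timer HUD.
; UpdateViewModel stores raw seconds, name and paused, started and urgency flags, and formats current time as minutes and zero-padded seconds.
; Display text prepends any non-empty label to the time.
; Rich text wraps that in a style tag, using a default style if the requested one is absent from AthenaStandaloneTimerStyles.
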(event Custom|UpdateViewModel
  (params (CurrentTime :type Integer) (TotalTime :type Integer) (LabelText :type Text%20%28by%20ref%29) (TextStyle :type Name) (TimerName :type Text%20%28by%20ref%29) (bIsPaused :type Boolean) (bIsStarted :type Boolean) (bIsUrgently :type Boolean))
  (Variables|Default|SetTotalTimeInSeconds TotalTime)
  (Variables|Default|SetCurrentTimeInSeconds CurrentTime)
  (Variables|Default|SetCurrentTime (CallFunction|GetFormatedTimeText CurrentTime))
  (Variables|Default|SetLabelText LabelText)
  (Variables|Default|SetDisplayText (Utilities|String|SelectText (Variables|Default|GetCurrentTime) (Utilities|Text|FormatText "NSLOCTEXT(\"[37CE7DD57BE3D9980E11680DDB7300BD]\", \"97979937478B98F388470D9621EE2D22\", \"{LabelText} {TimeText}\")" (Variables|Default|GetLabelText) (Variables|Default|GetCurrentTime)) (Utilities|Text|TextIsEmpty LabelText)))
  (Variables|Default|SetDisplayRichText (Utilities|Text|FormatText "NSLOCTEXT(\"[37CE7DD57BE3D9980E11680DDB7300BD]\", \"4727B78B4CEDCD83129B52BCF89A9AC9\", \"<{Style}>{Text}</>\")" (CallFunction|GetRichTextStyle TextStyle) (Variables|Default|GetDisplayText)))
  (Variables|Default|SetIsUrgency bIsUrgently)
  (Variables|Default|SetIsPaused bIsPaused)
  (Variables|Default|SetIsStarted bIsStarted)
  (Variables|Default|SetTimerName TimerName))

(fn GetFormatedTimeText
  (params (Time :type Integer))
  (return (Utilities|Text|FormatText "NSLOCTEXT(\"[37CE7DD57BE3D9980E11680DDB7300BD]\", \"79AEC2654F32AADFEEC66C891C1E8DFC\", \"{Minutes}:{Seconds}\")" (/ Time 60) (Utilities|String|Right (Utilities|String|Append "0" (Utilities|String|ToString(Integer) (Math|Integer|%25%28Integer%29 Time 60))) 2))))

(fn GetRichTextStyle
  (params (Style :type Name))
  (bind _outrownames (DataTable|GetDataTableRowNames "/Game/Athena/HUD/TextStyles/AthenaStandaloneTimerStyles.AthenaStandaloneTimerStyles"))
  (return (Utilities|String|SelectName Style (LocalVariables|GetDefaultStyle) (Utilities|Array|ContainsItem _outrownames Style))))
