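# Default timer widget for the Creative Timer device, a UMG layout that shows a time readout such as "88:88".
# A minimum-width SizeBox holds an Overlay that stacks a faint dark skewed-button background behind a centered rich text block.
# TimeRichText is exposed as a variable and uses the Athena standalone timer style set with a Heist tag on its placeholder.
# The graph holds only empty events, so nothing in this widget updates the time.
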WidgetBlueprint /CreativeCoreDevices/Timer/CreativeTimerWidget_Default
graphs [EventGraph {nodes [
   Event PreConstruct()
   Event Construct()
   Event Tick()
  ]}]
variables []
widgets {
 info {inheritedWidgetCount=0, parentClass=&UserWidget, rootWidgetClass=&SizeBox, widgetCount=4}
 widgets [
  SizeBox_0 {widgetClassPath=&SizeBox, widget=SizeBox_0, properties {bOverride_MinDesiredWidth=true, minDesiredWidth=106}}
  Overlay_0 {
   widgetClassPath=&Overlay
   parent=SizeBox_0
   widget=Overlay_0
   slot {class=SizeBoxSlot, properties {horizontalAlignment=HAlign_Center, verticalAlignment=VAlign_Center}}
  }
  Image_background {
   widgetClassPath=&Image
   parent=Overlay_0
   widget=Image_background
   properties {
    brush {
     bIsDynamicallyLoaded=false
     drawAs=Box
     imageType=NoImage
     mirroring=NoMirror
     resourceName=None
     resourceObject=/FortUILibrary/Foundation/Buttons/Materials/M_SkewButtonShape_DF
     tiling=NoTile
     imageSize=<88.8474 46.9781>
     margin=<ltrb 0.05 0.05 0.05 0.05>
     outlineSettings {
      bUseBrushTransparency=false
      roundingType=HalfHeightRadius
      width=0
      color {colorUseRule=UseColor_Specified, specifiedColor=<rgba 0 0 0 0>}
      cornerRadii=<0 0 0 0>
     }
     tintColor {colorUseRule=UseColor_Specified, specifiedColor=<rgba 1 1 1 1>}
     uVRegion {bIsValid=false, max=<0 0>, min=<0 0>}
    }
    colorAndOpacity=<rgba 0.0052 0.0052 0.0056 0.15>
   }
   slot {class=OverlaySlot, properties {horizontalAlignment=HAlign_Fill, verticalAlignment=VAlign_Fill, padding=<ltrb 0 -5 0 -3>}}
  }
  TimeRichText {
   widgetClassPath=/Script/UMG.RichTextBlock
   bIsVariable=true
   parent=Overlay_0
   widget=TimeRichText
   properties {justification=Center, text="<Heist>88:88</>", textStyleSet=/Game/Athena/HUD/TextStyles/AthenaStandaloneTimerStyles}
   slot {class=OverlaySlot, properties {horizontalAlignment=HAlign_Center, padding=<ltrb 12 0 10 0>}}
  }
 ]
}
